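# Physics asset for the CinderGale male medium player body, built around a multi-panel skirt with capsule colliders on the spine, pelvis and legs.
# Front, middle, side-back and back bone chains per side hang from tiny deform_skirt anchor spheres constrained to spine_02 and spine_03.
# Cross-constraints tie neighbouring chains into a sheet, joining left and right only at the back.
# Extra offset thigh and right-calf capsules and a huge sphere below root add collision volume.
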PhysicsAsset /BRCosmetics/Characters/Player/Male/Medium/Bodies/M_MED_CinderGale/Meshes/M_MED_CinderGale_Physics
bodies {
 calf_l [{shapeType=Capsule, length=40.5679, radius=9.856, center=<-20.1212 0 2.1399>, rotation=<rot 90 0 0>}]
 calf_r [
  {shapeType=Capsule, length=38.5679, radius=10.106, center=<20.0879 0 -2.1399>, rotation=<rot 90 -177.0178 0>}
  {shapeType=Capsule, length=26.25, radius=12, shapeName=calf_r_capsule, center=<27.094 3.1405 -3.0629>, rotation=<rot 90 6.4662 0>}
 ]
 deform_skirt_b_l [{shapeType=Sphere, radius=0.505, shapeName=deform_skirt_b_l_sphere}]
 deform_skirt_b_r [{shapeType=Sphere, radius=0.505, shapeName=deform_skirt_b_r_sphere}]
 deform_skirt_bk_l [{shapeType=Sphere, radius=0.505, shapeName=deform_skirt_bk_l_sphere}]
 deform_skirt_bk_r [{shapeType=Sphere, radius=0.505, shapeName=deform_skirt_bk_r_sphere}]
 deform_skirt_f_l [{shapeType=Sphere, radius=0.505, shapeName=deform_skirt_f_l_sphere}]
 deform_skirt_f_r [{shapeType=Sphere, radius=0.505, shapeName=deform_skirt_f_r_sphere}]
 deform_skirt_m_l [{shapeType=Sphere, radius=0.505, shapeName=deform_skirt_m_l_sphere}]
 deform_skirt_m_r [{shapeType=Sphere, radius=0.505, shapeName=deform_skirt_m_r_sphere}]
 dyn_skirt_b_1_l [{shapeType=Capsule, length=15.2574, radius=8.7427, shapeName=dyn_skirt_b_1_l_capsule, center=<10.0849 -7.3486 0>}]
 dyn_skirt_b_1_r [{shapeType=Capsule, length=19.8348, radius=8.7427, shapeName=dyn_skirt_b_1_r_capsule, center=<10.0849 -7.3486 0>, rotation=<rot 0 0 -180>}]
 dyn_skirt_b_2_l [{shapeType=Capsule, length=20.9798, radius=8.7427, shapeName=dyn_skirt_b_2_l_capsule, center=<10.0849 -5.5469 0>}]
 dyn_skirt_b_2_r [{shapeType=Capsule, length=26.4798, radius=8.7427, shapeName=dyn_skirt_b_2_r_capsule, center=<10.0849 -6.7298 0>, rotation=<rot 0 0 -180>}]
 dyn_skirt_b_l [{shapeType=Capsule, length=10.8532, radius=9.4063, shapeName=dyn_skirt_b_l_capsule, center=<7.3357 -7.3486 0>, rotation=<rot 90 0 0>}]
 dyn_skirt_b_r [{shapeType=Capsule, length=10.8532, radius=9.4063, shapeName=dyn_skirt_b_r_capsule, center=<7.3357 -7.3486 0>, rotation=<rot 90 0 0>}]
 dyn_skirt_bk_1_l [{shapeType=Capsule, length=13.6118, radius=10.3882, shapeName=dyn_skirt_bk_1_l_capsule, center=<8 -6.2367 0>}]
 dyn_skirt_bk_1_r [{shapeType=Capsule, length=13.6118, radius=10.3882, shapeName=dyn_skirt_bk_1_r_capsule, center=<8 -6.2367 0>}]
 dyn_skirt_bk_2_l [{shapeType=Capsule, length=20.9798, radius=8.7427, shapeName=dyn_skirt_bk_2_l_capsule, center=<10.0849 -4.435 0>}]
 dyn_skirt_bk_2_r [{shapeType=Capsule, length=26.4798, radius=8.7427, shapeName=dyn_skirt_bk_2_r_capsule, center=<10.0849 -4.435 0>}]
 dyn_skirt_bk_l [{
   shapeType=Capsule
   length=17.1137
   radius=9.4761
   shapeName=dyn_skirt_bk_l_capsule
   center=<8.3205 -7.3486 0>
   rotation=<rot 0.0002 0.0006 5.625>
  }]
 dyn_skirt_bk_r [{
   shapeType=Capsule
   length=17.1137
   radius=9.4761
   shapeName=dyn_skirt_bk_r_capsule
   center=<8.3205 -7.3486 0>
   rotation=<rot -0.0001 0.0002 -5.625>
  }]
 dyn_skirt_f_1_l [{shapeType=Capsule, length=15.2595, radius=5, shapeName=dyn_skirt_f_1_l_capsule, center=<7.3357 -2.9328 -2.3776>, rotation=<rot 90 0 0>}]
 dyn_skirt_f_1_r [{shapeType=Capsule, length=15.2595, radius=5, shapeName=dyn_skirt_f_1_r_capsule, center=<7.3357 -2.9328 2.3776>, rotation=<rot 90 180 0>}]
 dyn_skirt_f_2_l [{shapeType=Capsule, length=16.0837, radius=7.9164, shapeName=dyn_skirt_f_2_l_capsule, center=<10.0849 -6.9598 -2.3776>, rotation=<rot 90 0 0>}]
 dyn_skirt_f_2_r [{
   shapeType=Capsule
   length=16.0837
   radius=7.9164
   shapeName=dyn_skirt_f_2_r_capsule
   center=<10.0849 -6.9598 2.3776>
   rotation=<rot 90 180 0>
  }]
 dyn_skirt_f_3_l [
  {shapeType=Capsule, length=16.8337, radius=7.1664, shapeName=dyn_skirt_f_3_l_capsule, center=<10.0849 -6.9598 -2.3776>, rotation=<rot 90 0 0>}
  {shapeType=Capsule, length=15.75, radius=7.25, shapeName=dyn_skirt_f_3_l_capsule, center=<0 -5.9672 -6.9857>}
 ]
 dyn_skirt_f_3_r [
  {
   shapeType=Capsule
   length=16.0837
   radius=7.9164
   shapeName=dyn_skirt_f_3_r_capsule
   center=<10.0849 -6.9598 2.3776>
   rotation=<rot 90 180 0>
  }
  {shapeType=Capsule, length=15, radius=8, shapeName=dyn_skirt_f_3_r_capsule, center=<0 -5.9672 6.9857>}
 ]
 dyn_skirt_f_l [{shapeType=Capsule, length=19, radius=5, shapeName=dyn_skirt_f_l_capsule, center=<10.0849 -2.9328 -2.3776>, rotation=<rot 90 0 0>}]
 dyn_skirt_f_r [{shapeType=Capsule, length=19, radius=5, shapeName=dyn_skirt_f_r_capsule, center=<10.0849 -2.9328 2.3776>, rotation=<rot 90 0 0>}]
 dyn_skirt_m_1_l [{shapeType=Capsule, length=15.2595, radius=5, shapeName=dyn_skirt_m_1_l_capsule, center=<7.3357 -2.9328 0>, rotation=<rot 90 0 0>}]
 dyn_skirt_m_1_r [{shapeType=Capsule, length=15.2595, radius=5, shapeName=dyn_skirt_m_1_r_capsule, center=<7.3357 -2.9328 0>, rotation=<rot 90 0 0>}]
 dyn_skirt_m_2_l [{shapeType=Capsule, length=16.0837, radius=7.9164, shapeName=dyn_skirt_m_2_l_capsule, center=<10.0849 -6.9598 0>}]
 dyn_skirt_m_2_r [{shapeType=Capsule, length=16.0837, radius=7.9164, shapeName=dyn_skirt_m_2_r_capsule, center=<10.0849 -6.9598 0>, rotation=<rot 0 0 -180>}]
 dyn_skirt_m_3_l [{shapeType=Capsule, length=16.8337, radius=7.1664, shapeName=dyn_skirt_m_3_l_capsule, center=<10.0849 -6.9598 0>}]
 dyn_skirt_m_3_r [{shapeType=Capsule, length=16.0837, radius=7.9164, shapeName=dyn_skirt_m_3_r_capsule, center=<10.0849 -8.1427 0>, rotation=<rot 0 0 -180>}]
 dyn_skirt_m_l [{shapeType=Capsule, length=19, radius=5, shapeName=dyn_skirt_m_l_capsule, center=<10.0849 -2.9328 0>, rotation=<rot 90 0 0>}]
 dyn_skirt_m_r [{shapeType=Capsule, length=19, radius=5, shapeName=dyn_skirt_m_r_capsule, center=<10.0849 -2.9328 0>, rotation=<rot 90 0 0>}]
 foot_l [{shapeType=Capsule, length=18.8559, radius=8.8599, center=<-3.6645 6.4179 0>, rotation=<rot 0 0 90>}]
 foot_r [{shapeType=Capsule, length=18.8559, radius=8.8599, center=<3.6645 -6.4179 0>, rotation=<rot 0 0 -90>}]
 pelvis [{shapeType=Capsule, length=3.6972, radius=11.5832}]
 root [{shapeType=Sphere, radius=139.771, shapeName=root_sphere, center=<0 -15.3649 -131.3441>}]
 spine_01 [{shapeType=Capsule, length=4.4524, radius=11.427, shapeName=spine_01_capsule}]
 spine_02 [{shapeType=Capsule, length=3.6972, radius=11.5832, center=<0.0003 0 0>}]
 spine_03 [{shapeType=Capsule, length=3.091, radius=11.5832}]
 spine_04 [{shapeType=Capsule, length=3.091, radius=11.5832}]
 spine_05 [{shapeType=Capsule, length=3.6972, radius=11.5832}]
 thigh_l [
  {shapeType=Capsule, length=37.2767, radius=9.2896, center=<-18.668 0.8409 0.5942>, rotation=<rot 90 0 0>}
  {shapeType=Capsule, length=26.9939, radius=10.0405, shapeName=thigh_l_capsule, center=<-19.9337 1.7136 6.9484>, rotation=<rot 90 0 0>}
 ]
 thigh_r [
  {shapeType=Capsule, length=37.8123, radius=8.7539, center=<18.668 0 0>, rotation=<rot 90 180 0>}
  {shapeType=Capsule, length=30.7625, radius=10.0405, shapeName=thigh_r_capsule, center=<19.7288 -5.4763 -6.9484>, rotation=<rot 90 175.9754 0>}
 ]
}
constraints [
 {bone1Name=spine_02, bone2Name=spine_01, swing1LimitDegrees=45, swing2LimitDegrees=45, twistLimitDegrees=45}
 {bone1Name=spine_03, bone2Name=spine_02, swing1LimitDegrees=45, swing2LimitDegrees=45, twistLimitDegrees=45}
 {bone1Name=spine_04, bone2Name=spine_03, swing1LimitDegrees=45, swing2LimitDegrees=45, twistLimitDegrees=45}
 {bone1Name=spine_05, bone2Name=spine_04, swing1LimitDegrees=45, swing2LimitDegrees=45, twistLimitDegrees=45}
 {bone1Name=thigh_l, bone2Name=pelvis, swing1LimitDegrees=45, swing2LimitDegrees=45, twistLimitDegrees=45}
 {bone1Name=calf_l, bone2Name=thigh_l, swing1LimitDegrees=45, swing2LimitDegrees=45, twistLimitDegrees=45}
 {bone1Name=foot_l, bone2Name=calf_l, swing1LimitDegrees=45, swing2LimitDegrees=45, twistLimitDegrees=45}
 {bone1Name=thigh_r, bone2Name=pelvis, swing1LimitDegrees=45, swing2LimitDegrees=45, twistLimitDegrees=45}
 {bone1Name=calf_r, bone2Name=thigh_r, swing1LimitDegrees=45, swing2LimitDegrees=45, twistLimitDegrees=45}
 {bone1Name=foot_r, bone2Name=calf_r, swing1LimitDegrees=45, swing2LimitDegrees=45, twistLimitDegrees=45}
 {
  bone1Name=spine_01
  bone2Name=pelvis
  swing1LimitDegrees=45
  swing1Motion=Free
  swing2LimitDegrees=45
  swing2Motion=Free
  twistLimitDegrees=45
  twistMotion=Free
 }
 {bone1Name=dyn_skirt_b_l, bone2Name=deform_skirt_b_l, swing1LimitDegrees=15, swing2LimitDegrees=15, twistLimitDegrees=5, twistMotion=Locked}
 {bone1Name=dyn_skirt_b_1_l, bone2Name=dyn_skirt_b_l, swing1LimitDegrees=15, swing2LimitDegrees=15, twistLimitDegrees=5}
 {bone1Name=dyn_skirt_b_2_l, bone2Name=dyn_skirt_b_1_l, swing1LimitDegrees=35, swing2LimitDegrees=35, twistLimitDegrees=15}
 {bone1Name=dyn_skirt_b_r, bone2Name=deform_skirt_b_r, swing1LimitDegrees=15, swing2LimitDegrees=15, twistLimitDegrees=5, twistMotion=Locked}
 {bone1Name=dyn_skirt_b_1_r, bone2Name=dyn_skirt_b_r, swing1LimitDegrees=15, swing2LimitDegrees=15, twistLimitDegrees=5}
 {bone1Name=dyn_skirt_b_2_r, bone2Name=dyn_skirt_b_1_r, swing1LimitDegrees=35, swing2LimitDegrees=35, twistLimitDegrees=15}
 {
  bone1Name=dyn_skirt_bk_l
  bone2Name=deform_skirt_bk_l
  swing1LimitDegrees=15
  swing2LimitDegrees=15
  swing2Motion=Locked
  twistLimitDegrees=5
  twistMotion=Locked
 }
 {bone1Name=dyn_skirt_bk_1_l, bone2Name=dyn_skirt_bk_l, swing1LimitDegrees=30, swing2LimitDegrees=30, twistLimitDegrees=15}
 {bone1Name=dyn_skirt_bk_2_l, bone2Name=dyn_skirt_bk_1_l, swing1LimitDegrees=15, swing2LimitDegrees=5, twistLimitDegrees=1}
 {
  bone1Name=dyn_skirt_bk_r
  bone2Name=deform_skirt_bk_r
  swing1LimitDegrees=15
  swing2LimitDegrees=15
  swing2Motion=Locked
  twistLimitDegrees=5
  twistMotion=Locked
 }
 {bone1Name=dyn_skirt_bk_1_r, bone2Name=dyn_skirt_bk_r, swing1LimitDegrees=30, swing2LimitDegrees=30, twistLimitDegrees=15}
 {bone1Name=dyn_skirt_bk_2_r, bone2Name=dyn_skirt_bk_1_r, swing1LimitDegrees=15, swing2LimitDegrees=5, twistLimitDegrees=1}
 {bone1Name=dyn_skirt_f_l, bone2Name=deform_skirt_f_l, swing1LimitDegrees=15, swing2LimitDegrees=15, twistLimitDegrees=5}
 {bone1Name=dyn_skirt_f_1_l, bone2Name=dyn_skirt_f_l, swing1LimitDegrees=15, swing2LimitDegrees=15, twistLimitDegrees=5}
 {bone1Name=dyn_skirt_f_2_l, bone2Name=dyn_skirt_f_1_l, swing1LimitDegrees=45, swing2LimitDegrees=15, twistLimitDegrees=5}
 {bone1Name=dyn_skirt_f_3_l, bone2Name=dyn_skirt_f_2_l, swing1LimitDegrees=45, swing2LimitDegrees=15, twistLimitDegrees=5}
 {bone1Name=dyn_skirt_f_r, bone2Name=deform_skirt_f_r, swing1LimitDegrees=15, swing2LimitDegrees=15, twistLimitDegrees=5}
 {bone1Name=dyn_skirt_f_1_r, bone2Name=dyn_skirt_f_r, swing1LimitDegrees=15, swing2LimitDegrees=15, twistLimitDegrees=5}
 {bone1Name=dyn_skirt_f_2_r, bone2Name=dyn_skirt_f_1_r, swing1LimitDegrees=45, swing2LimitDegrees=15, twistLimitDegrees=5}
 {bone1Name=dyn_skirt_f_3_r, bone2Name=dyn_skirt_f_2_r, swing1LimitDegrees=45, swing2LimitDegrees=15, twistLimitDegrees=5}
 {bone1Name=dyn_skirt_m_l, bone2Name=deform_skirt_m_l, swing1LimitDegrees=15, swing2LimitDegrees=15, twistLimitDegrees=5}
 {bone1Name=dyn_skirt_m_1_l, bone2Name=dyn_skirt_m_l, swing1LimitDegrees=15, swing2LimitDegrees=15, twistLimitDegrees=5}
 {bone1Name=dyn_skirt_m_2_l, bone2Name=dyn_skirt_m_1_l, swing1LimitDegrees=15, swing2LimitDegrees=15, twistLimitDegrees=5}
 {bone1Name=dyn_skirt_m_3_l, bone2Name=dyn_skirt_m_2_l, swing1LimitDegrees=45, swing2LimitDegrees=30, twistLimitDegrees=5}
 {bone1Name=dyn_skirt_m_r, bone2Name=deform_skirt_m_r, swing1LimitDegrees=15, swing2LimitDegrees=15, twistLimitDegrees=5}
 {bone1Name=dyn_skirt_m_1_r, bone2Name=dyn_skirt_m_r, swing1LimitDegrees=15, swing2LimitDegrees=15, twistLimitDegrees=5}
 {bone1Name=dyn_skirt_m_2_r, bone2Name=dyn_skirt_m_1_r, swing1LimitDegrees=15, swing2LimitDegrees=15, twistLimitDegrees=5}
 {bone1Name=dyn_skirt_m_3_r, bone2Name=dyn_skirt_m_2_r, swing1LimitDegrees=45, swing2LimitDegrees=30, twistLimitDegrees=5}
 {bone1Name=dyn_skirt_bk_l, bone2Name=dyn_skirt_b_l, swing1LimitDegrees=15, swing2LimitDegrees=10, twistLimitDegrees=5}
 {
  bone1Name=dyn_skirt_bk_1_l
  bone2Name=dyn_skirt_b_1_l
  swing1LimitDegrees=15
  swing1Motion=Free
  swing2LimitDegrees=10
  swing2Motion=Free
  twistLimitDegrees=5
  twistMotion=Free
 }
 {bone1Name=dyn_skirt_bk_r, bone2Name=dyn_skirt_b_r, swing1LimitDegrees=15, swing2LimitDegrees=10, twistLimitDegrees=5}
 {
  bone1Name=dyn_skirt_bk_1_r
  bone2Name=dyn_skirt_b_1_r
  swing1LimitDegrees=15
  swing1Motion=Free
  swing2LimitDegrees=10
  swing2Motion=Free
  twistLimitDegrees=5
  twistMotion=Free
 }
 {bone1Name=dyn_skirt_m_1_l, bone2Name=dyn_skirt_f_1_l, swing1LimitDegrees=30, swing2LimitDegrees=15, twistLimitDegrees=5}
 {bone1Name=dyn_skirt_m_2_l, bone2Name=dyn_skirt_f_2_l, swing1LimitDegrees=30, swing2LimitDegrees=15, twistLimitDegrees=5}
 {bone1Name=dyn_skirt_m_l, bone2Name=dyn_skirt_f_l, swing1LimitDegrees=30, swing2LimitDegrees=15, twistLimitDegrees=5}
 {bone1Name=dyn_skirt_m_r, bone2Name=dyn_skirt_f_r, swing1LimitDegrees=30, swing2LimitDegrees=15, twistLimitDegrees=5}
 {bone1Name=dyn_skirt_m_1_r, bone2Name=dyn_skirt_f_1_r, swing1LimitDegrees=30, swing2LimitDegrees=15, twistLimitDegrees=5}
 {bone1Name=dyn_skirt_m_2_r, bone2Name=dyn_skirt_f_2_r, swing1LimitDegrees=30, swing2LimitDegrees=15, twistLimitDegrees=5}
 {bone1Name=dyn_skirt_bk_r, bone2Name=dyn_skirt_bk_l, swing1LimitDegrees=15, swing2LimitDegrees=10, twistLimitDegrees=5}
 {bone1Name=dyn_skirt_bk_1_r, bone2Name=dyn_skirt_bk_1_l, swing1LimitDegrees=15, swing2LimitDegrees=10, twistLimitDegrees=5}
 {
  bone1Name=dyn_skirt_b_l
  bone2Name=dyn_skirt_m_1_l
  swing1LimitDegrees=30
  swing1Motion=Free
  swing2LimitDegrees=30
  swing2Motion=Free
  twistLimitDegrees=30
  twistMotion=Free
 }
 {
  bone1Name=dyn_skirt_b_r
  bone2Name=dyn_skirt_m_1_r
  swing1LimitDegrees=30
  swing1Motion=Free
  swing2LimitDegrees=30
  swing2Motion=Free
  twistLimitDegrees=30
  twistMotion=Free
 }
 {bone1Name=dyn_skirt_bk_2_r, bone2Name=dyn_skirt_bk_2_l, swing1LimitDegrees=25, swing2LimitDegrees=15, twistLimitDegrees=10}
 {bone1Name=dyn_skirt_bk_2_l, bone2Name=dyn_skirt_b_2_l, swing1LimitDegrees=15, swing2LimitDegrees=15, twistLimitDegrees=10, twistMotion=Free}
 {bone1Name=dyn_skirt_bk_2_r, bone2Name=dyn_skirt_b_2_r, swing1LimitDegrees=15, swing2LimitDegrees=15, twistLimitDegrees=15}
 {
  bone1Name=dyn_skirt_b_1_r
  bone2Name=dyn_skirt_m_2_r
  swing1LimitDegrees=30
  swing1Motion=Free
  swing2LimitDegrees=30
  swing2Motion=Free
  twistLimitDegrees=30
  twistMotion=Free
 }
 {
  bone1Name=dyn_skirt_b_2_l
  bone2Name=dyn_skirt_m_3_l
  swing1LimitDegrees=15
  swing1Motion=Free
  swing2LimitDegrees=15
  swing2Motion=Free
  twistLimitDegrees=35
  twistMotion=Free
 }
 {
  bone1Name=dyn_skirt_b_2_r
  bone2Name=dyn_skirt_m_3_r
  swing1LimitDegrees=15
  swing1Motion=Free
  swing2LimitDegrees=15
  swing2Motion=Free
  twistLimitDegrees=15
  twistMotion=Free
 }
 {
  bone1Name=pelvis
  bone2Name=root
  swing1LimitDegrees=45
  swing1Motion=Free
  swing2LimitDegrees=45
  swing2Motion=Free
  twistLimitDegrees=45
  twistMotion=Free
 }
 {bone1Name=dyn_skirt_m_2_l, bone2Name=dyn_skirt_b_1_l, swing1LimitDegrees=15, swing2LimitDegrees=15, twistLimitDegrees=15, twistMotion=Free}
 {bone1Name=dyn_skirt_m_3_r, bone2Name=dyn_skirt_f_3_r, swing1LimitDegrees=15, swing2LimitDegrees=15, twistLimitDegrees=15}
 {
  bone1Name=deform_skirt_b_l
  bone2Name=spine_02
  swing1LimitDegrees=45
  swing1Motion=Free
  swing2LimitDegrees=45
  swing2Motion=Free
  twistLimitDegrees=45
  twistMotion=Free
 }
 {
  bone1Name=deform_skirt_b_r
  bone2Name=spine_02
  swing1LimitDegrees=45
  swing1Motion=Free
  swing2LimitDegrees=45
  swing2Motion=Free
  twistLimitDegrees=45
  twistMotion=Free
 }
 {
  bone1Name=deform_skirt_bk_l
  bone2Name=spine_02
  swing1LimitDegrees=45
  swing1Motion=Free
  swing2LimitDegrees=45
  swing2Motion=Free
  twistLimitDegrees=45
  twistMotion=Free
 }
 {
  bone1Name=deform_skirt_bk_r
  bone2Name=spine_02
  swing1LimitDegrees=45
  swing1Motion=Free
  swing2LimitDegrees=45
  swing2Motion=Free
  twistLimitDegrees=45
  twistMotion=Free
 }
 {
  bone1Name=deform_skirt_f_l
  bone2Name=spine_03
  swing1LimitDegrees=45
  swing1Motion=Free
  swing2LimitDegrees=45
  swing2Motion=Free
  twistLimitDegrees=45
  twistMotion=Free
 }
 {
  bone1Name=deform_skirt_f_r
  bone2Name=spine_03
  swing1LimitDegrees=45
  swing1Motion=Free
  swing2LimitDegrees=45
  swing2Motion=Free
  twistLimitDegrees=45
  twistMotion=Free
 }
 {
  bone1Name=deform_skirt_m_l
  bone2Name=spine_03
  swing1LimitDegrees=45
  swing1Motion=Free
  swing2LimitDegrees=45
  swing2Motion=Free
  twistLimitDegrees=45
  twistMotion=Free
 }
 {
  bone1Name=deform_skirt_m_r
  bone2Name=spine_03
  swing1LimitDegrees=45
  swing1Motion=Free
  swing2LimitDegrees=45
  swing2Motion=Free
  twistLimitDegrees=45
  twistMotion=Free
 }
 {bone1Name=dyn_skirt_m_3_l, bone2Name=dyn_skirt_f_3_l, swing1LimitDegrees=15, swing2LimitDegrees=15, twistLimitDegrees=15}
]
