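# Cosmetic AnimBP for the CinderGale male character body, copying the attached parent's pose and adding secondary motion.
# A Control Rig pose feeds three RigidBody sims, selected by baseball-slide, front-end or default state, then a post Control Rig.
# Anim Dynamics swings three zippers, and a skirt bone briefly rotates when crouching starts.
# Event Graph updates gravity (via GnatGala parameters) and sim-space variables, but the RigidBody nodes use fixed values.
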AnimBlueprint /BRCosmetics/Characters/Player/Male/Medium/Bodies/M_MED_CinderGale/Meshes/M_MED_CinderGale_AnimBP
target_skeleton=&/M_MED_CinderGale_Skeleton
graphs [
 AnimGraph {nodes [
   n1 = RigidBody!(
    ComponentPose=Local To Component(LocalPose=Use cached pose 'pre rigid'())
    Alpha=Get Alpha() -> (Alpha:float)
    OverrideWorldGravity:Vector="0,0,-980"
    SimSpaceSettings:Sim Space Settings Structure=(WorldAlpha=1,VelocityScaleZ=0.3,DampingAlpha=1,MaxLinearVelocity=5000,MaxAngularVelocity=5000,MaxLinearAcceleration=1500,MaxAngularAcceleration=500,ExternalLinearDragV=(X=0.2,Y=0.2,Z=0),ExternalLinearVelocity=(X=0,Y=0,Z=0),ExternalAngularVelocity=(X=0,Y=0,Z=0))
   ) -> (Pose) {
    bClampLinearTranslationLimitToRefPose=true
    bOverrideWorldGravity=true
    overridePhysicsAsset=&/M_MED_CinderGale_FE_Physics
    simulationSpace=BaseBoneSpace
    baseBoneRef {boneName=spine_02}
    overrideWorldGravity {z=-980}
    simSpaceSettings {
     maxAngularAcceleration=500
     maxAngularVelocity=5000
     maxLinearAcceleration=1500
     maxLinearVelocity=5000
     velocityScaleZ=0.3
     worldAlpha=1
     externalLinearDragV=<0.2 0.2>
    }
   }
   n2 = Component To Local(ComponentPose=@n1)
   n3 = RigidBody(
    ComponentPose=Local To Component(LocalPose=Use cached pose 'pre rigid'())
    Alpha=Get Alpha()
    OverrideWorldGravity="0,0,-980"
    SimSpaceSettings=(WorldAlpha=1,VelocityScaleZ=0.3,DampingAlpha=1,MaxLinearVelocity=5000,MaxAngularVelocity=5000,MaxLinearAcceleration=1500,MaxAngularAcceleration=500,ExternalLinearDragV=(X=0.2,Y=0.2,Z=0),ExternalLinearVelocity=(X=0,Y=0,Z=0),ExternalAngularVelocity=(X=0,Y=0,Z=0))
   ) {
    bClampLinearTranslationLimitToRefPose=true
    bOverrideWorldGravity=true
    overridePhysicsAsset=&/M_MED_CinderGale_Physics
    simulationSpace=BaseBoneSpace
    baseBoneRef {boneName=spine_02}
    overrideWorldGravity {z=-980}
    simSpaceSettings {
     maxAngularAcceleration=500
     maxAngularVelocity=5000
     maxLinearAcceleration=1500
     maxLinearVelocity=5000
     velocityScaleZ=0.3
     worldAlpha=1
     externalLinearDragV=<0.2 0.2>
    }
   }
   n4 = Component To Local(ComponentPose=@n3)
   n5 = Get bIsInFrontEnd()
   n6 = RigidBody(
    ComponentPose=Local To Component(LocalPose=Use cached pose 'pre rigid'())
    Alpha=Get Alpha()
    OverrideWorldGravity="0,0,-980"
    SimSpaceSettings=(WorldAlpha=1,VelocityScaleZ=0.3,DampingAlpha=1,MaxLinearVelocity=5000,MaxAngularVelocity=5000,MaxLinearAcceleration=1500,MaxAngularAcceleration=500,ExternalLinearDragV=(X=0.2,Y=0.2,Z=0),ExternalLinearVelocity=(X=0,Y=0,Z=0),ExternalAngularVelocity=(X=0,Y=0,Z=0))
   ) {
    bClampLinearTranslationLimitToRefPose=true
    bOverrideWorldGravity=true
    overridePhysicsAsset=&/M_MED_CinderGale_SlidePhysics
    simulationSpace=BaseBoneSpace
    baseBoneRef {boneName=spine_02}
    overrideWorldGravity {z=-980}
    simSpaceSettings {
     maxAngularAcceleration=500
     maxAngularVelocity=5000
     maxLinearAcceleration=1500
     maxLinearVelocity=5000
     velocityScaleZ=0.3
     worldAlpha=1
     externalLinearDragV=<0.2 0.2>
    }
   }
   n7 = Control Rig!(
    Source=Blend Poses by bool(
     BlendPose_0=Component To Local(ComponentPose=@n6)
     BlendPose_1=Blend Poses by bool(BlendPose_0=@n2, BlendPose_1=@n4, bActiveValue=@n5)
     bActiveValue=Get bIsBaseballSliding()
    )
    Alpha
    Emotes:bool=Get bIsPlayingEmote()
    DBNO:bool=Get bIsDBNO()
   ) -> (Pose) {controlRigAssetReference {blueprintRigClass=&/M_MED_CinderGale_Post_CtrlRig.M_MED_CinderGale_Post_CtrlRig_C}}
   n8 = Anim Dynamics - Bone: dyn_zipper_l!(ComponentPose=Local To Component(LocalPose=@n7), Alpha=1) -> (Pose) {
    angularDampingOverride=0.99
    angularSpringConstant=1
    bAngularSpring=true
    bOverrideAngularDamping=true
    simulationSpace=RootRelative
    boundBone {boneName=dyn_zipper_l}
    physicsBodyDefinitions [{
      boundBone {boneName=dyn_zipper_l}
      boxExtents=<10 3 5>
      constraintSetup {angularLimitsMax=<0 45 -15>, angularLimitsMin=<0 0 -90>, angularTarget=<0 0 0>}
      localJointOffset=<0.3 0 0>
     }]
   }
   n9 = Anim Dynamics - Bone: dyn_zipper_r!(ComponentPose=@n8, Alpha=1) -> (Pose) {
    angularDampingOverride=0.99
    angularSpringConstant=1
    bAngularSpring=true
    bOverrideAngularDamping=true
    simulationSpace=RootRelative
    boundBone {boneName=dyn_zipper_r}
    physicsBodyDefinitions [{
      boundBone {boneName=dyn_zipper_r}
      boxExtents=<10 3 5>
      constraintSetup {angularLimitsMax=<0 0 -10>, angularLimitsMin=<0 -45 -90>, angularTarget=<0 0 0>}
      localJointOffset=<0.3 0 0>
     }]
   }
   n10 = Anim Dynamics - Bone: dyn_zipper_f!(ComponentPose=@n9, Alpha=1) -> (Pose) {
    angularDampingOverride=0.98
    angularSpringConstant=1
    bAngularSpring=true
    bOverrideAngularDamping=true
    simulationSpace=RootRelative
    boundBone {boneName=dyn_zipper_f}
    physicsBodyDefinitions [{
      boundBone {boneName=dyn_zipper_f}
      boxExtents=<10 3 5>
      constraintSetup {angularLimitsMax=<0 45 -10>, angularLimitsMin=<0 -45 -90>, angularTarget=<0 0 0>}
      localJointOffset=<0.3 0 0>
     }]
   }
   n11 = "Transform (Modify) Bone - Bone: dyn_skirt_f_1_l"!(
    ComponentPose=@n10
    bAlphaBoolEnabled:bool=Get CrouchingClipping() -> (CrouchingClipping:bool)
    Translation="0,0,0"
    Rotation="-10,-10,0"
    Scale="1,1,1"
   ) -> (Pose) {
    alphaInputType=Bool
    rotationMode=BMM_Additive
    rotationSpace=BCS_BoneSpace
    alphaBoolBlend {blendInTime=0.1, blendOutTime=0.2}
    boneToModify {boneName=dyn_skirt_f_1_l}
    rotation {pitch=-10, yaw=-10}
   }
   Output Pose(Result=Component To Local(ComponentPose=@n11)) {layerGroup=None}
   n12 = Control Rig!(
    Source=Copy Pose From Mesh() {bUseAttachedParent=true}
    Alpha
    Sliding:bool=Get bIsBaseballSliding()
    Skydive:bool=Get bIsSkyDiving()
    Emotes=Get bIsPlayingEmote()
    InFrtontend:bool=Get bIsInFrontEnd()
   ) -> (Pose) {controlRigAssetReference {blueprintRigClass=&/M_MED_CinderGale_CtrlRig.M_MED_CinderGale_CtrlRig_C}}
   Save cached pose 'pre rigid'(Pose=@n12) {cacheName=pre rigid}
  ]}
 EventGraph {nodes [
   Event BlueprintUpdateAnimation()
   > n1 = Sequence()
   > n2 = "Gravity Override State Function (Parameter Asset) (Nativized)"(
    CharacterPartAnimInstance=Self-Reference()
    GameTimeInSeconds=GetGameTimeInSeconds()
    Parameters=/BRCosmetics/Characters/Player/Male/Medium/Bodies/M_MED_GnatGala/Meshes/M_MED_GnatGala_Parameters
   )
   > Set Gravity Override(Gravity Override=@n2)
   > n3 = Branch(Condition=Get bIsInFrontEnd())
   > Set!(
    execute
    Sim Space Settings_WorldAlpha:float=0.3
    Sim Space Settings_VelocityScaleZ:float=0.2
    Sim Space Settings_DampingAlpha:float=1
    Sim Space Settings_MaxLinearVelocity:float=2500
    Sim Space Settings_MaxAngularVelocity:float=2500
    Sim Space Settings_MaxLinearAcceleration:float=1500
    Sim Space Settings_MaxAngularAcceleration:float=500
    Sim Space Settings_ExternalLinearDragV:Vector="1,1,1"
    Sim Space Settings_ExternalLinearVelocity:Vector="0,0,0"
    Sim Space Settings_ExternalAngularVelocity:Vector="0,0,0.1"
   ) -> (then, Output_Get:Sim Space Settings Structure)
   Set(
    execute=@n3.else
    Sim Space Settings_WorldAlpha=0.4
    Sim Space Settings_VelocityScaleZ=0.4
    Sim Space Settings_DampingAlpha=1
    Sim Space Settings_MaxLinearVelocity=2500
    Sim Space Settings_MaxAngularVelocity=2500
    Sim Space Settings_MaxLinearAcceleration=1500
    Sim Space Settings_MaxAngularAcceleration=500
    Sim Space Settings_ExternalLinearDragV="1,1,1"
    Sim Space Settings_ExternalLinearVelocity="0,0,0"
    Sim Space Settings_ExternalAngularVelocity="0,0,0.1"
   )
   n4 = Branch(execute=@n1.then_1, Condition=Get bIsCrouching())
   > Do Once(Reset=@n4.else)
   > Set CrouchingClipping(CrouchingClipping:bool=true) -> (Output_Get:bool)
   > Delay(Duration=0.2)
   > Set CrouchingClipping(CrouchingClipping=false)
   Event BlueprintBeginPlay() {is=title K2Node_Event}
   > Set Alpha(Alpha:float=1) -> (Output_Get:float)
   Set(
    Sim Space Settings_WorldAlpha=0.07
    Sim Space Settings_VelocityScaleZ=0
    Sim Space Settings_DampingAlpha=1
    Sim Space Settings_MaxLinearVelocity=500
    Sim Space Settings_MaxAngularVelocity=3
    Sim Space Settings_MaxLinearAcceleration=700
    Sim Space Settings_MaxAngularAcceleration=250
    Sim Space Settings_ExternalLinearDragV="0.1,0.1,0"
    Sim Space Settings_ExternalLinearVelocity="0,0,0"
    Sim Space Settings_ExternalAngularVelocity="0,0,0.1"
   )
  ]}
]
variables [
 Alpha {type=number}
 CrouchingClipping {type=boolean}
 Gravity Override {type=Vector}
 Sim Space Settings {
  type=SimSpaceSettings
  default {
   dampingAlpha=1
   maxAngularAcceleration=10000
   maxAngularVelocity=10000
   maxLinearAcceleration=10000
   maxLinearVelocity=10000
   velocityScaleZ=1
   worldAlpha=0
   externalLinearDragV=<0 0 0>
  }
 }
]
