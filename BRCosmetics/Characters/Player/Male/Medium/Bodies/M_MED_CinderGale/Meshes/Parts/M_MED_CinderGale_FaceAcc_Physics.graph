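# Physics asset for the CinderGale medium male face accessory mesh, setting up dynamic hair bodies alongside head and upper-body shapes.
# Five hair chains (back, back-left/right, middle-left/right) each run from a small head-linked root through two larger segments with angular limits.
# The middle chains lock rotation at their first joint, and cross-constraints tie neighboring tips together so the hair moves as a group.
# Head, neck, upper spine, clavicle and upper-arm capsules approximate the body, joined by free-rotation constraints.
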PhysicsAsset /BRCosmetics/Characters/Player/Male/Medium/Bodies/M_MED_CinderGale/Meshes/Parts/M_MED_CinderGale_FaceAcc_Physics
bodies {
 clavicle_l [{
   shapeType=Capsule
   length=12.8749
   radius=8.7281
   shapeName=clavicle_l_capsule
   center=<6.4453 -1.3287 -0.8523>
   rotation=<rot 73.125 -0.0027 179.9971>
  }]
 clavicle_r [{
   shapeType=Capsule
   length=12.8749
   radius=8.7281
   shapeName=clavicle_r_capsule
   center=<-6.4453 1.3287 0.8523>
   rotation=<rot 73.125 -0.0027 -0.0029>
  }]
 deform_hair_b_l [{shapeType=Capsule, length=0.505, radius=0.505, shapeName=deform_hair_b_l_capsule, rotation=<rot 0 -90 0>}]
 deform_hair_b_r [{shapeType=Capsule, length=0.505, radius=0.505, shapeName=deform_hair_b_r_capsule, rotation=<rot 0 -90 0>}]
 deform_hair_bk [{shapeType=Capsule, length=0.505, radius=0.505, shapeName=deform_hair_bk_capsule, rotation=<rot 0 -90 0>}]
 deform_hair_m_l [{shapeType=Capsule, length=0.505, radius=0.505, shapeName=deform_hair_m_l_capsule, rotation=<rot 0 -90 0>}]
 deform_hair_m_r [{shapeType=Capsule, length=0.505, radius=0.505, shapeName=deform_hair_m_r_capsule, rotation=<rot 0 -90 0>}]
 dyn_hair_1_b_l [{shapeType=Capsule, length=6, radius=3.3, shapeName=dyn_hair_1_b_l_capsule, center=<3.5292 -1.2295 0>}]
 dyn_hair_1_b_r [{shapeType=Capsule, length=6, radius=3.3, shapeName=dyn_hair_1_b_r_capsule, center=<3.7575 -1.1418 0>}]
 dyn_hair_1_bk [{shapeType=Capsule, length=6, radius=3.3, shapeName=dyn_hair_1_bk_capsule, center=<3.7838 -1.4329 0>}]
 dyn_hair_1_m_l [
  {shapeType=Sphere, radius=1.9856, shapeName=dyn_hair_1_m_l_sphere, center=<3.0828 2.0382 2.811>}
  {shapeType=Capsule, length=4.6929, radius=2.3253, shapeName=dyn_hair_1_m_l_capsule, center=<2.0098 0.8892 0.1474>, rotation=<rot 0 0 -180>}
 ]
 dyn_hair_1_m_r [
  {shapeType=Sphere, radius=1.771, shapeName=dyn_hair_1_m_r_sphere, center=<2.589 1.9592 -3.6187>}
  {shapeType=Capsule, length=4.6929, radius=2.3253, shapeName=dyn_hair_1_m_r_capsule, center=<2.0098 0.8892 0.1474>}
 ]
 dyn_hair_b_l [{shapeType=Capsule, length=6, radius=3.3, shapeName=dyn_hair_b_l_capsule, center=<2.6433 -1.0431 0>}]
 dyn_hair_b_r [{shapeType=Capsule, length=6, radius=3.3, shapeName=dyn_hair_b_r_capsule, center=<2.6433 -1.0431 0>}]
 dyn_hair_bk [{shapeType=Capsule, length=6, radius=3.3, shapeName=dyn_hair_bk_capsule, center=<2.6433 -1.0431 0>}]
 dyn_hair_m_l [{shapeType=Capsule, length=7.0845, radius=2.2155, shapeName=dyn_hair_m_l_capsule, center=<2.6433 -1.0431 0>, rotation=<rot 0 0 -180>}]
 dyn_hair_m_r [{shapeType=Capsule, length=7.0845, radius=2.2155, shapeName=dyn_hair_m_r_capsule, center=<2.6433 -1.0431 0>}]
 head [{shapeType=Capsule, length=7.479, radius=6.7046, center=<3.6218 2.0604 0>, rotation=<rot 90 11.2498 0>}]
 neck_01 [{shapeType=Capsule, length=3.7826, radius=6.1042, shapeName=neck_01_capsule, center=<2.5527 0 0>, rotation=<rot 90 0 0>}]
 neck_02 [{shapeType=Capsule, length=4.8868, radius=5, shapeName=neck_02_capsule, center=<2.5527 0 0>, rotation=<rot 90 0 0>}]
 spine_05 [{shapeType=Capsule, length=10.3452, radius=13.0539, shapeName=spine_05_capsule, center=<-2.1212 1.9013 0>, rotation=<rot 90 0 0>}]
 upperarm_l [{shapeType=Capsule, length=16.8974, radius=6.2362, shapeName=upperarm_l_capsule, center=<6.6094 0 0>, rotation=<rot 90 180 0>}]
 upperarm_r [{shapeType=Capsule, length=16.8974, radius=6.2362, shapeName=upperarm_r_capsule, center=<-6.6094 0 0>, rotation=<rot 90 0 0>}]
}
constraints [
 {
  bone1Name=head
  bone2Name=neck_02
  swing1LimitDegrees=45
  swing1Motion=Free
  swing2LimitDegrees=45
  swing2Motion=Free
  twistLimitDegrees=45
  twistMotion=Free
 }
 {
  bone1Name=neck_02
  bone2Name=neck_01
  swing1LimitDegrees=45
  swing1Motion=Free
  swing2LimitDegrees=45
  swing2Motion=Free
  twistLimitDegrees=45
  twistMotion=Free
 }
 {
  bone1Name=neck_01
  bone2Name=spine_05
  swing1LimitDegrees=45
  swing1Motion=Free
  swing2LimitDegrees=45
  swing2Motion=Free
  twistLimitDegrees=45
  twistMotion=Free
 }
 {
  bone1Name=clavicle_l
  bone2Name=spine_05
  swing1LimitDegrees=45
  swing1Motion=Free
  swing2LimitDegrees=45
  swing2Motion=Free
  twistLimitDegrees=45
  twistMotion=Free
 }
 {
  bone1Name=clavicle_r
  bone2Name=spine_05
  swing1LimitDegrees=45
  swing1Motion=Free
  swing2LimitDegrees=45
  swing2Motion=Free
  twistLimitDegrees=45
  twistMotion=Free
 }
 {
  bone1Name=upperarm_l
  bone2Name=clavicle_l
  swing1LimitDegrees=45
  swing1Motion=Free
  swing2LimitDegrees=45
  swing2Motion=Free
  twistLimitDegrees=45
  twistMotion=Free
 }
 {
  bone1Name=upperarm_r
  bone2Name=clavicle_r
  swing1LimitDegrees=45
  swing1Motion=Free
  swing2LimitDegrees=45
  swing2Motion=Free
  twistLimitDegrees=45
  twistMotion=Free
 }
 {bone1Name=dyn_hair_b_l, bone2Name=deform_hair_b_l, swing1LimitDegrees=60, swing2LimitDegrees=10, twistLimitDegrees=5}
 {bone1Name=dyn_hair_1_b_l, bone2Name=dyn_hair_b_l, swing1LimitDegrees=40, swing2LimitDegrees=15, twistLimitDegrees=5}
 {bone1Name=dyn_hair_b_r, bone2Name=deform_hair_b_r, swing1LimitDegrees=60, swing2LimitDegrees=10, twistLimitDegrees=5}
 {bone1Name=dyn_hair_1_b_r, bone2Name=dyn_hair_b_r, swing1LimitDegrees=40, swing2LimitDegrees=15, twistLimitDegrees=5}
 {bone1Name=dyn_hair_bk, bone2Name=deform_hair_bk, swing1LimitDegrees=60, swing2LimitDegrees=10, twistLimitDegrees=5}
 {bone1Name=dyn_hair_1_bk, bone2Name=dyn_hair_bk, swing1LimitDegrees=40, swing2LimitDegrees=15, twistLimitDegrees=5}
 {
  bone1Name=dyn_hair_m_l
  bone2Name=deform_hair_m_l
  swing1LimitDegrees=60
  swing1Motion=Locked
  swing2LimitDegrees=15
  swing2Motion=Locked
  twistLimitDegrees=5
  twistMotion=Locked
 }
 {bone1Name=dyn_hair_1_m_l, bone2Name=dyn_hair_m_l, swing1LimitDegrees=60, swing2LimitDegrees=5, twistLimitDegrees=30}
 {
  bone1Name=dyn_hair_m_r
  bone2Name=deform_hair_m_r
  swing1LimitDegrees=60
  swing1Motion=Locked
  swing2LimitDegrees=15
  swing2Motion=Locked
  twistLimitDegrees=5
  twistMotion=Locked
 }
 {bone1Name=dyn_hair_1_m_r, bone2Name=dyn_hair_m_r, swing1LimitDegrees=60, swing2LimitDegrees=5, twistLimitDegrees=30}
 {bone1Name=dyn_hair_1_bk, bone2Name=dyn_hair_1_b_r, swing1LimitDegrees=15, swing2LimitDegrees=15, twistLimitDegrees=15}
 {bone1Name=dyn_hair_1_bk, bone2Name=dyn_hair_1_b_l, swing1LimitDegrees=15, swing2LimitDegrees=15, twistLimitDegrees=15}
 {bone1Name=dyn_hair_1_b_l, bone2Name=dyn_hair_1_m_l, swing1LimitDegrees=20, swing2LimitDegrees=20, twistLimitDegrees=20}
 {bone1Name=dyn_hair_1_b_r, bone2Name=dyn_hair_1_m_r, swing1LimitDegrees=20, swing2LimitDegrees=20, twistLimitDegrees=20}
 {
  bone1Name=deform_hair_b_l
  bone2Name=head
  swing1LimitDegrees=45
  swing1Motion=Free
  swing2LimitDegrees=45
  swing2Motion=Free
  twistLimitDegrees=45
  twistMotion=Free
 }
 {
  bone1Name=deform_hair_b_r
  bone2Name=head
  swing1LimitDegrees=45
  swing1Motion=Free
  swing2LimitDegrees=45
  swing2Motion=Free
  twistLimitDegrees=45
  twistMotion=Free
 }
 {
  bone1Name=deform_hair_bk
  bone2Name=head
  swing1LimitDegrees=45
  swing1Motion=Free
  swing2LimitDegrees=45
  swing2Motion=Free
  twistLimitDegrees=45
  twistMotion=Free
 }
 {
  bone1Name=deform_hair_m_l
  bone2Name=head
  swing1LimitDegrees=45
  swing1Motion=Free
  swing2LimitDegrees=45
  swing2Motion=Free
  twistLimitDegrees=45
  twistMotion=Free
 }
 {
  bone1Name=deform_hair_m_r
  bone2Name=head
  swing1LimitDegrees=45
  swing1Motion=Free
  swing2LimitDegrees=45
  swing2Motion=Free
  twistLimitDegrees=45
  twistMotion=Free
 }
]
