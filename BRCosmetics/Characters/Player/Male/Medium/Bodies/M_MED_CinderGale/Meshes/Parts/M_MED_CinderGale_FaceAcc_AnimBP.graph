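# Animation Blueprint for the Cinder Gale medium male skin's face accessory mesh, adding secondary motion to its dynamic parts, including front hair bones.
# It copies the parent mesh's pose, runs the accessory's Control Rig, then applies a RigidBody simulation in neck_02 base-bone space using a dedicated physics asset.
# Three chained Anim Dynamics nodes then swing the front, front-left and front-right hair bones with damped, angle-limited springs.
# The event graph holds no logic.
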AnimBlueprint /BRCosmetics/Characters/Player/Male/Medium/Bodies/M_MED_CinderGale/Meshes/Parts/M_MED_CinderGale_FaceAcc_AnimBP
target_skeleton=&/M_MED_CinderGale_FaceAcc_Skeleton
graphs [
 AnimGraph {nodes [
   n1 = Control Rig(Source=Copy Pose From Mesh() {bUseAttachedParent=true}) {controlRigAssetReference {blueprintRigClass=&/M_MED_CinderGale_FaceAcc_CtrlRig.M_MED_CinderGale_FaceAcc_CtrlRig_C}}
   n2 = RigidBody(ComponentPose=Local To Component(LocalPose=@n1)) {
    bClampLinearTranslationLimitToRefPose=true
    bOverrideWorldGravity=true
    overridePhysicsAsset=&/M_MED_CinderGale_FaceAcc_Physics
    simulationSpace=BaseBoneSpace
    baseBoneRef {boneName=neck_02}
    overrideWorldGravity {z=-980}
    simSpaceSettings {
     maxAngularAcceleration=250
     maxAngularVelocity=5000
     maxLinearAcceleration=500
     maxLinearVelocity=5000
     velocityScaleZ=0.3
     worldAlpha=0.5
     externalLinearDragV=<0.2 0.2 0.2>
    }
   }
   n3 = Anim Dynamics - Bone: dyn_hair_f(ComponentPose=@n2, Alpha=1, ExternalForce="0,0,0") {
    angularDampingOverride=0.98
    angularSpringConstant=50
    bAngularSpring=true
    bOverrideAngularDamping=true
    simulationSpace=RootRelative
    boundBone {boneName=dyn_hair_f}
    physicsBodyDefinitions [{
      sphereCollisionRadius=3
      boundBone {boneName=dyn_hair_f}
      boxExtents=<10 5 5>
      constraintSetup {angularLimitsMax=<0 15 -3>, angularLimitsMin=<0 -15 -35>, angularTarget=<0 0 0>}
      localJointOffset=<0.2 0 0>
     }]
    relativeSpaceBone {boneName=neck_01}
   }
   n4 = Anim Dynamics - Bone: dyn_hair_f_l(ComponentPose=@n3, Alpha=1, ExternalForce="0,0,0") {
    angularDampingOverride=0.98
    angularSpringConstant=100
    bAngularSpring=true
    bOverrideAngularDamping=true
    simulationSpace=RootRelative
    boundBone {boneName=dyn_hair_f_l}
    physicsBodyDefinitions [{
      sphereCollisionRadius=3
      boundBone {boneName=dyn_hair_f_l}
      boxExtents=<10 5 5>
      constraintSetup {angularLimitsMax=<0 15 -3>, angularLimitsMin=<0 -15 -35>, angularTarget=<0 0 0>}
      localJointOffset=<0.1 0 0>
     }]
    relativeSpaceBone {boneName=neck_01}
   }
   n5 = Anim Dynamics - Bone: dyn_hair_f_r(ComponentPose=@n4, Alpha=1, ExternalForce="0,0,0") {
    angularDampingOverride=0.98
    angularSpringConstant=80
    bAngularSpring=true
    bOverrideAngularDamping=true
    simulationSpace=RootRelative
    boundBone {boneName=dyn_hair_f_r}
    physicsBodyDefinitions [{
      sphereCollisionRadius=3
      boundBone {boneName=dyn_hair_f_r}
      boxExtents=<10 5 5>
      constraintSetup {angularLimitsMax=<0 15 -3>, angularLimitsMin=<0 -15 -35>, angularTarget=<0 0 0>}
      localJointOffset=<0.15 0 0>
     }]
    relativeSpaceBone {boneName=neck_01}
   }
   Output Pose(Result=Component To Local(ComponentPose=@n5)) {layerGroup=None}
  ]}
 EventGraph {nodes [Event BlueprintUpdateAnimation()]}
]
variables []
